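# Default Lit material for the Fortanium prism quest telescope prop, faking a dispersive crystal with emissive refraction.
# Three refracted skylight environment samples at offset IORs supply separate RGB channels for chromatic aberration, tinted by the diffuse texture.
# A Fresnel mask blends this with MF_FortaniumVeins, and a camera-relative animated sweep provides a glint.
# World position offset combines switchable bobbing and rotation, with matching normal rotation, plus a black hole deformation function.
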Material /Game/Environments/Asteria/Props/Quest/Quest_Telescope/Materials/M_Fortanium_Prism
graph {
 properties {bUseMaterialAttributes=true, pixelDepthOffsetMode=PDOM_Legacy, usedShadingModels=Default Lit}
 nodes [
  n1 = NamedRerouteUsage()
  n2 = ScalarParameter() {defaultValue=1.5, group=Prisim, parameterName=IOR}
  n3 = ScalarParameter() {defaultValue=0.3, group=Prisim, parameterName=Abberation}
  n4 = Multiply(A=CameraVectorWS(), B=Constant3Vector() {constant=<rgba 1 1 -1 0>}) {constB=-1}
  n5 = PixelNormalWS()
  n6 = ScalarParameter() {defaultValue=0.1, group=Prisim, parameterName=RefractonRoughness}
  n7 = SkyLightEnvMapSample(
   Direction=MaterialFunctionCall!(Ray Direction=@n4, Surface Normal=@n5, Refractive Index Origin, Refractive Index Target=Subtract(A=@n2, B=@n3)) {materialFunction=/Engine/Functions/Engine_MaterialFunctions01/Vectors/Refract}.Vector
   Roughness=@n6
  )
  n8 = ScalarParameter() {defaultValue=1, group=Prisim, parameterName=PreAbberationDesaturate}
  n9 = SkyLightEnvMapSample(
   Direction=MaterialFunctionCall(Ray Direction=@n4, Surface Normal=@n5, Refractive Index Target=@n2) {materialFunction=&Refract}.Vector
   Roughness=@n6
  )
  n10 = SkyLightEnvMapSample(
   Direction=MaterialFunctionCall(Ray Direction=@n4, Surface Normal=@n5, Refractive Index Target=Add(A=@n2, B=@n3)) {materialFunction=&Refract}.Vector
   Roughness=@n6
  )
  n11 = MaterialFunctionCall(
   X=ComponentMask(None=Desaturation(None=@n7, Fraction=@n8)) {r=true}
   Y=ComponentMask(None=Desaturation(None=@n9, Fraction=@n8)) {g=true}
   Z=ComponentMask(None=Desaturation(None=@n10, Fraction=@n8)) {b=true}
  ) {materialFunction=&MakeFloat3}
  n12 = TextureSampleParameter2D() {parameterName=Diffuse, texture=&T_WhitesRGB}
  n13 = LinearInterpolate(A=@n11.Result, B=Multiply(A=@n11.Result, B=@n12.RGB), Alpha=ScalarParameter() {defaultValue=0.6, group=Prisim, parameterName=TintReflection})
  n14 = TextureSampleParameter2D() {parameterName=Normal, samplerType=SAMPLERTYPE_Normal, texture=/Engine/EngineMaterials/FlatNormal}
  n15 = MaterialFunctionCall(BaseNormal=@n14.RGB, AdditionalNormal=NamedRerouteUsage()) {materialFunction=&BlendAngleCorrectedNormals}
  n16 = SetMaterialAttributes!(
   MaterialAttributes
   Emissive Color=Multiply(A=Add(A=@n1, B=@n13), B=ScalarParameter() {defaultValue=1, group=Prisim, parameterName=EmissiveMultiplier})
   Normal=@n15.Result
   Base Color=Multiply(A=@n12.RGB, B=ScalarParameter() {defaultValue=1, group=Prisim, parameterName=BaseColorMultiplier})
  ) {attributeSetTypes [#1, #2, #3]}
  n17 = MaterialFunctionCall!(VeinMask=Constant() {r=1}, Material, Position=WorldPosition() {worldPositionShaderOffset=WPT_ExcludeAllShaderOffsets}.XYZ) {materialFunction=/Game/Athena/Asteria/Environments/Landscape/Materials/MaterialFunctions/MF_FortaniumVeins}
  n18 = Fresnel(ExponentIn=ScalarParameter() {defaultValue=4, group=Prisim, parameterName=FortaniumFresnelExponent}) {baseReflectFraction=0, exponent=2}
  Composite!(A=@n16, B=@n17.Result, Alpha=@n18) {subgraphName=Blend Materials}
  Constant() {r=0.001}
  n19 = Constant3Vector() {constant=<rgba 0 0 1 0>}
  n20 = Multiply(A=Time(), B=ScalarParameter() {defaultValue=0.4, parameterName=WPOTimeScale})
  n21 = Multiply(
   A=LinearInterpolate(
    A=ScalarParameter() {defaultValue=-0.01, parameterName=Rotation Angle Min}
    B=ScalarParameter() {defaultValue=0.01, parameterName=Rotation Angle Max}
    Alpha=Cosine(None=@n20)
   )
   B=OneMinus(None=VectorParameter() {parameterName=WorldPostionOfBlackHole, defaultValue=<rgba -30000 25536 1952 0>}.A)
  )
  n22 = ScalarParameter() {defaultValue=2, parameterName=NormalRotateExaggeration}
  NamedRerouteDeclaration(None=MaterialFunctionCall(Rotation Axis=@n19, Rotation Angle=Multiply(A=@n21, B=@n22)) {materialFunction=&FixRotateAboutAxisNormals}.New Tangent Space Vertex Normal) {name=RotateNormals, nodeColor=<rgba 1 0 0.3047 1>}
  n23 = Multiply(A=Subtract(A=WorldPosition().XYZ, B=ObjectPositionWS()), B=ScalarParameter() {defaultValue=0.001, parameterName=WorldTileScale})
  n24 = Normalize(VectorInput=Transform(None=VectorParameter() {parameterName=SweepDirectionCamSpace, defaultValue=<rgba -0.5 0.45 -1 0>}.RGB) {transformSourceType=TRANSFORMSOURCE_Camera})
  n25 = ComponentMask(None=Multiply(A=DotProduct(A=@n23, B=@n24), B=@n24)) {b=true, r=true}
  n26 = Constant() {r=0.28}
  n27 = Frac(None=ComponentMask(None=Panner(Coordinate=@n25, Speed=@n26) {speedX=0.2, speedY=0.2}) {g=true})
  n28 = Multiply(A=Subtract(A=@n27) {constB=0.9}) {constB=10}
  n29 = Multiply(A=Floor(None=Add(A=@n27) {constB=0.1}), B=Power(Base=OneMinus(None=@n28)) {constExponent=3})
  n30 = Saturate(None=Add(A=MaterialFunctionCall() {materialFunction=&MF_Fresnel_FX}.Fresnel Inverted) {constB=0.1})
  n31 = VectorParameter() {parameterName=SweepColour, defaultValue=<rgba 1.8958 2 1.9078 0>}
  n32 = Multiply(A=@n31.RGB, B=ScalarParameter() {defaultValue=15, parameterName=FBoostScale})
  n33 = VertexColor()
  NamedRerouteDeclaration(None=Multiply(A=Multiply(A=@n29, B=@n30), B=LinearInterpolate(A=@n31.RGB, B=@n32, Alpha=@n33.G))) {name=Sweep, nodeColor=<rgba 0 1 0.6484 1>}
  MaterialExpressionReroute_12 = Reroute(None=@n16)
  n34 = GetMaterialAttributes(None=@MaterialExpressionReroute_12.A) {attributeGetTypes [
    #3
    #4
    #5
    #6
    #1
   ]}
  MaterialExpressionReroute_13 = Reroute(None=@n17.Result)
  n35 = GetMaterialAttributes(None=@MaterialExpressionReroute_13.B) {attributeGetTypes [
    #3
    #4
    #5
    #6
    #1
   ]}
  MaterialExpressionReroute_14 = Reroute(None=@n18)
  MaterialExpressionAppend3Vector_1 = MaterialXAppend3Vector!(A=@n34.Metallic, B=@n34.Metallic, C=@n34.Metallic)
  MaterialExpressionAppend3Vector_0 = MaterialXAppend3Vector(A=@n35.Metallic, B=@n35.Metallic, C=@n35.Metallic)
  n36 = MaterialFunctionCall(Float3=LinearInterpolate(A=@MaterialExpressionAppend3Vector_1, B=@MaterialExpressionAppend3Vector_0, Alpha=@MaterialExpressionReroute_14.Alpha)) {materialFunction=&BreakOutFloat3Components}
  n37 = SetMaterialAttributes!(
   MaterialAttributes=@n34.MaterialAttributes
   Base Color=LinearInterpolate(A=@n34.Base Color, B=@n35.Base Color, Alpha=@MaterialExpressionReroute_14.Alpha)
   Metallic=@n36.R
   Specular=@n36.R
   Roughness=@n36.R
   Emissive Color=LinearInterpolate(A=@n34.Emissive Color, B=@n35.Emissive Color, Alpha=@MaterialExpressionReroute_14.Alpha)
  ) {attributeSetTypes [
    #3
    #4
    #5
    #6
    #1
   ]}
  PinBase!(Output=@n37) {reroutePins [Output {expression=&self:MaterialExpressionReroute_15}]}
  PinBase!() {reroutePins [
    A {expression=&self:MaterialExpressionReroute_12}
    B {expression=&self:MaterialExpressionReroute_13}
    Alpha {expression=&self:MaterialExpressionReroute_14}
   ]}
  MaterialExpressionReroute_15 = Reroute(None=@n37)
  n38 = Add(
   A=LinearInterpolate(
    A=VectorParameter() {parameterName=BobHeightMax, defaultValue=<rgba 0 0 10 0>}.RGB
    B=VectorParameter() {parameterName=BobHeightMin, defaultValue=<rgba 0 0 -2 0>}.RGB
    Alpha=Sine(None=@n20)
   )
   B=RotateAboutAxis(NormalizedRotationAxis=@n19, RotationAngle=@n21, PivotPoint=ObjectPositionWS(), Position=WorldPosition().XYZ)
  )
  n39 = Constant()
  n40 = VectorParameter() {parameterName=WorldPostionOfBlackHole, defaultValue=<rgba -30000 25536 1952 0>}
  n41 = AppendVector(A=@n40.RGB, B=@n40.RGB)
  SetMaterialAttributes!(
   MaterialAttributes=@MaterialExpressionReroute_15.Output
   World Position Offset=Add(
    A=StaticSwitchParameter(True=@n38, False=@n39) {parameterName=Use WPO}
    B=MaterialFunctionCall!(WorldPostionOfBlackHole=@n41, World Position) {materialFunction=/Game/Packages/Fortnite_UniqueMaterials_1/BlackHoleFunction}.Result
   )
  ) {attributeSetTypes [#7]}
  StaticSwitchParameter(True=@n15.Result, False=@n14.RGB) {parameterName=Use WPO}
 ]
}
